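# Sound cue for a looping House of Mirrors rain ambience, filed under the Creative radio gadget's survival horror sound loops.
# A single looping wave player feeds the output directly, with no randomization or mixing nodes.
# The cue boosts volume and uses tile ambient attenuation and the Creative ambient tile mix sound class.
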SoundCue /Game/Sounds/Creative/Gadgets/Radio/Sound_Loops/Suvival_Horror/HouseOfMirrors/Amb_HouseOfMirrors_Rain_Loop_Cue
root=WavePlayer_0
nodes [{
  id=WavePlayer_0
  properties {bLooping=true, soundWaveAssetPtr=&/Amb_HouseOfMirrors_Rain_Loop}
 }]
properties {
 attenuationSettings=/Game/Sounds/Attenuation/Fort_TileAmbient_Att
 maxDistance=10000
 soundClassObject=/Game/Sounds/FortSoundClassesAndModes/Fort_Creative_Ambient_TileMix
 volumeMultiplier=3
}
